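# Physics asset for the Textile Knight backpack cosmetic, giving its handle and five zipper bones simulated-physics bodies and joints.
# The handle is constrained to a small pack capsule with limited swing and twist.
# Each zipper pull, a small capsule or box, hangs from one of four larger master capsules, two sharing master C, with one free swing axis and tight twist.
# The master capsules have no constraints to the pack.
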PhysicsAsset /BRCosmetics/Accessories/FORT_Backpacks/M_MED_Textile_Knight_Backpack/Meshes/M_MED_Textile_Knight_Pack_Physics
bodies {
 dyn_handle [{shapeType=Capsule, length=3, radius=2, shapeName=dyn_handle_capsule, center=<1.3 0 0>, rotation=<rot 0 0 90>}]
 dyn_zipperMSTR_A [{shapeType=Capsule, length=6, radius=2.9, shapeName=dyn_zipperMSTR_D_capsule, center=<0 0.75 -3.25>, rotation=<rot 85 0 0>}]
 dyn_zipperMSTR_B [{shapeType=Capsule, length=6, radius=2.9, shapeName=dyn_zipperMSTR_D_capsule, center=<0 0.3 -3.5>, rotation=<rot 85 0 0>}]
 dyn_zipperMSTR_C [{shapeType=Capsule, length=8, radius=5, shapeName=dyn_zipperMSTR_D_capsule, center=<0 0 -5.25>, rotation=<rot 80 0 0>}]
 dyn_zipperMSTR_D [{shapeType=Capsule, length=6, radius=2.9, shapeName=dyn_zipperMSTR_D_capsule, center=<0 -0.75 -3.25>, rotation=<rot 85 0 0>}]
 dyn_zipper_01 [{shapeType=Capsule, length=1.5, radius=0.5, shapeName=dyn_zipper_05_capsule, center=<2 0 0>, rotation=<rot 90 0 0>}]
 dyn_zipper_02 [{shapeType=Capsule, length=1.5, radius=0.5, shapeName=dyn_zipper_05_capsule, center=<2 0 0>, rotation=<rot 90 0 0>}]
 dyn_zipper_03 [{shapeType=Box, extentX=1.352, extentY=0.5, extentZ=2.5, shapeName=dyn_zipper_03_box, center=<2 0 0>, rotation=<rot 0 90 90>}]
 dyn_zipper_04 [{shapeType=Box, extentX=1.3204, extentY=0.3, extentZ=2.5, shapeName=dyn_zipper_04_box, center=<2 0 0>, rotation=<rot 0 90 90>}]
 dyn_zipper_05 [{shapeType=Capsule, length=1.5, radius=0.5, shapeName=dyn_zipper_05_capsule, center=<2 0 0>, rotation=<rot 90 0 0>}]
 pack [{shapeType=Capsule, length=1, radius=1, shapeName=pack_capsule}]
}
constraints [
 {bone1Name=dyn_zipper_01, bone2Name=dyn_zipperMSTR_A, swing1LimitDegrees=45, swing2LimitDegrees=45, swing2Motion=Free, twistLimitDegrees=5}
 {bone1Name=dyn_zipper_02, bone2Name=dyn_zipperMSTR_B, swing1LimitDegrees=45, swing2LimitDegrees=45, swing2Motion=Free, twistLimitDegrees=5}
 {bone1Name=dyn_zipper_05, bone2Name=dyn_zipperMSTR_D, swing1LimitDegrees=45, swing2LimitDegrees=45, swing2Motion=Free, twistLimitDegrees=5}
 {bone1Name=dyn_handle, bone2Name=pack, swing1LimitDegrees=5, swing2LimitDegrees=60, twistLimitDegrees=15}
 {bone1Name=dyn_zipper_03, bone2Name=dyn_zipperMSTR_C, swing1LimitDegrees=45, swing2LimitDegrees=45, swing2Motion=Free, twistLimitDegrees=5}
 {bone1Name=dyn_zipper_04, bone2Name=dyn_zipperMSTR_C, swing1LimitDegrees=45, swing2LimitDegrees=45, swing2Motion=Free, twistLimitDegrees=5}
]
