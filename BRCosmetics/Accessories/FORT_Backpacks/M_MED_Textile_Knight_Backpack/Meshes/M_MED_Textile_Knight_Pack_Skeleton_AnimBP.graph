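# Animation Blueprint for the Textile Knight backpack cosmetic, adding physics-driven secondary motion to the pack.
# It copies the pose from a mesh and applies a small additive bone-space offset to the pack bone.
# A RigidBody node then simulates the pack's override physics asset relative to the root bone, with scaled component velocity and acceleration and overridden world gravity.
# The EventGraph is empty.
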AnimBlueprint /BRCosmetics/Accessories/FORT_Backpacks/M_MED_Textile_Knight_Backpack/Meshes/M_MED_Textile_Knight_Pack_Skeleton_AnimBP
target_skeleton=&/M_MED_Textile_Knight_Pack_Skeleton
graphs [
 AnimGraph {nodes [
   n1 = "Transform (Modify) Bone - Bone: pack"(ComponentPose=Local To Component(LocalPose=Copy Pose From Mesh()), Translation="0,2.5,0", Rotation="0,0,0") {
    translationMode=BMM_Additive
    translationSpace=BCS_BoneSpace
    alphaBoolBlend {blendOption=Linear}
    boneToModify {boneName=pack}
    translation {y=2.5}
   }
   n2 = RigidBody!(ComponentPose=@n1, Alpha, OverrideWorldGravity:Vector="0,0,-980", ExternalForce) -> (Pose) {
    bClampLinearTranslationLimitToRefPose=true
    bOverrideWorldGravity=true
    overridePhysicsAsset=&/M_MED_Textile_Knight_Pack_Physics
    simulationSpace=BaseBoneSpace
    alphaBoolBlend {blendOption=Linear}
    baseBoneRef {boneName=root}
    componentAppliedLinearAccClamp=<3000 3000 3000>
    componentLinearAccScale=<0.4 0.3 0.35>
    componentLinearVelScale=<0.4 0.3 0.35>
   }
   Output Pose(Result=Component To Local(ComponentPose=@n2)) {layerGroup=None}
  ]}
 EventGraph {links [], nodes []}
]
variables [
 Gravity Override Params {type=GravityOverrideParamsStruct}
 gravityOverride {type=Vector}
]
